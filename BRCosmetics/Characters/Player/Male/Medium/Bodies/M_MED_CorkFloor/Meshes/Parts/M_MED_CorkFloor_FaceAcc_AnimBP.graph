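# Animation Blueprint for the CorkFloor male face accessory, driving back-hair bones in response to head pose and worn back bling.
# Copies parent and body-mesh poses, then rotates hair by head-up and head-turn alphas.
# Selects RigidBody simulation without back bling, Control Rig plus RigidBody with it, or Control Rig plus aiming/crouching bone offsets for capes or long back bling.
# The event graph updates gravity, skydive scales, back-bling flags and weapon-based alphas.
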AnimBlueprint /BRCosmetics/Characters/Player/Male/Medium/Bodies/M_MED_CorkFloor/Meshes/Parts/M_MED_CorkFloor_FaceAcc_AnimBP
target_skeleton=&/M_MED_CorkFloor_FaceAcc_Skeleton
graphs [
 AnimGraph {nodes [
   n1 = RigidBody!(
    ComponentPose=Local To Component(LocalPose=Use cached pose 'Cache_BackpackState'())
    Alpha
    OverrideWorldGravity:Vector="0,0,-980"
    ExternalForce
    ComponentLinearAccScale:Vector="0.01,0.01,0.01"
    ComponentLinearVelScale:Vector="0.01,0.01,0.01"
   ) -> (Pose) {
    bClampLinearTranslationLimitToRefPose=true
    bOverrideWorldGravity=true
    overridePhysicsAsset=&/M_MED_CorkFloor_FaceAcc_Physics
    simulationSpace=BaseBoneSpace
    alphaBoolBlend {blendOption=Linear}
    baseBoneRef {boneName=neck_01}
    componentAppliedLinearAccClamp=<2000 2000 2000>
    componentLinearAccScale=<0.6 0.2 0.2>
    componentLinearVelScale=<0.6 0.2 0.2>
    simSpaceSettings {worldAlpha=0.01}
   }
   n2 = Component To Local(ComponentPose=@n1)
   n3 = Get Linear Scale Free() -> (Linear Scale Free:Vector)
   n4 = RigidBody(
    ComponentPose=Local To Component(LocalPose=Use cached pose 'Cache_StartPose'())
    OverrideWorldGravity=Get GravityOverride()
    ComponentLinearAccScale=@n3
    ComponentLinearVelScale=@n3
   ) {
    bClampLinearTranslationLimitToRefPose=true
    bOverrideWorldGravity=true
    overridePhysicsAsset=&M_MED_CorkFloor_FaceAcc_Physics
    simulationSpace=BaseBoneSpace
    alphaBoolBlend {blendOption=Linear}
    baseBoneRef {boneName=neck_01}
    componentAppliedLinearAccClamp=<2000 2000 2000>
    componentLinearAccScale=<0.6 0.2 0.2>
    componentLinearVelScale=<0.6 0.2 0.2>
    simSpaceSettings {worldAlpha=0.05}
   }
   n5 = Component To Local(ComponentPose=@n4)
   n6 = Get isWearingBackbling()
   n7 = Get AimingUp_ShortW() -> (AimingUp_ShortW:double)
   n8 = "Transform (Modify) Bone - Bone: base_hair_back_r"(
    ComponentPose=Local To Component(LocalPose=Use cached pose 'Cache_BackpackState'())
    Alpha=@n7
    Translation="2,1,0"
    Rotation="0,0,0"
    Scale="1,1,1"
   ) {
    rotationMode=BMM_Additive
    rotationSpace=BCS_BoneSpace
    translationMode=BMM_Additive
    translationSpace=BCS_BoneSpace
    alphaBoolBlend {blendOption=Linear}
    alphaScaleBiasClamp {bInterpResult=true}
    boneToModify {boneName=base_hair_back_r}
    translation=<2 1>
   }
   n9 = "Transform (Modify) Bone - Bone: base_hair_back_m"(ComponentPose=@n8, Alpha=@n7, Translation="3,1,0", Rotation="0,0,0", Scale="1,1,1") {
    rotationMode=BMM_Additive
    rotationSpace=BCS_BoneSpace
    translationMode=BMM_Additive
    translationSpace=BCS_BoneSpace
    alphaBoolBlend {blendOption=Linear}
    alphaScaleBiasClamp {bInterpResult=true}
    boneToModify {boneName=base_hair_back_m}
    translation=<3 1>
   }
   n10 = "Transform (Modify) Bone - Bone: base_hair_back_r"(
    ComponentPose=@n9
    Alpha=Get Crouching_BackpackCape_SW() -> (Crouching_BackpackCape_SW:double)
    Translation="0,4,0"
    Rotation="0,-5,0"
    Scale="1,1,1"
   ) {
    rotationMode=BMM_Additive
    rotationSpace=BCS_BoneSpace
    scaleMode=BMM_Additive
    scaleSpace=BCS_BoneSpace
    translationMode=BMM_Additive
    translationSpace=BCS_BoneSpace
    alphaBoolBlend {blendOption=Linear}
    alphaScaleBiasClamp {bInterpResult=true}
    boneToModify {boneName=base_hair_back_r}
    rotation {yaw=-5}
    translation {y=4}
   }
   n11 = "Transform (Modify) Bone - Bone: base_hair_back_r"(
    ComponentPose=@n10
    Alpha=Get Crouching_BackpackCape_LW() -> (Crouching_BackpackCape_LW:double)
    Translation="-1,5,0"
    Rotation="0,-2,0"
    Scale="1,1,1"
   ) {
    rotationMode=BMM_Additive
    rotationSpace=BCS_BoneSpace
    scaleMode=BMM_Additive
    scaleSpace=BCS_BoneSpace
    translationMode=BMM_Additive
    translationSpace=BCS_BoneSpace
    alphaBoolBlend {blendOption=Linear}
    alphaScaleBiasClamp {bInterpResult=true}
    boneToModify {boneName=base_hair_back_r}
    rotation {yaw=-2}
    translation=<-1 5>
   }
   Output Pose(Result=Blend Poses by bool(
     BlendPose_0=Component To Local(ComponentPose=@n11)
     BlendPose_1=Blend Poses by bool(BlendPose_0=@n2, BlendPose_1=@n5, bActiveValue=@n6) {blendType=Linear}
     bActiveValue=Get Is Wearing Cape() -> (Is Wearing Cape:bool)
    ) {blendType=Linear}) {layerGroup=None}
   n12 = Control Rig!(Source=Use cached pose 'Cache_StartPose'(), Alpha, parachuting_alpha:double=Get isParachuting_alpha() -> (isParachuting_alpha:double)) -> (Pose) {
    alphaBoolBlend {blendOption=Linear}
    controlRigAssetReference {blueprintRigClass=&/F_MED_CorkFloor_FaceAcc_CtrlRig.F_MED_CorkFloor_FaceAcc_CtrlRig_C}
   }
   Save cached pose 'Cache_BackpackState'(Pose=@n12) {cacheName=Cache_BackpackState}
   n13 = Get Head_Turn_L() -> (Head_Turn_L:double)
   n14 = Get Head_Turn_R() -> (Head_Turn_R:double)
   n15 = Layered blend per bone(
    BasePose=Copy Pose From Mesh() {bCopyCurves=true, bUseAttachedParent=true}
    BlendPoses_0=Copy Pose From Mesh(SourceMeshComponent=Get Body_Mesh())
   ) {layerSetup [{branchFilters [
       {boneName=clavicle_l}
       {boneName=clavicle_r}
       {boneName=upperarm_l}
       {boneName=upperarm_r}
       {boneName=upperarm_twist_01_l}
       {boneName=upperarm_twist_01_r}
       {boneName=deform_upperarm_bk_base_l}
       {boneName=deform_upperarm_bk_l}
       {boneName=deform_upperarm_fr_base_l}
       {boneName=deform_upperarm_fr_l}
       {boneName=deform_upperarm_up_base_l}
       {boneName=deform_upperarm_up_l}
       {boneName=deform_upperarm_bk_base_r}
       {boneName=deform_upperarm_bk_r}
       {boneName=deform_upperarm_fr_base_r}
       {boneName=deform_upperarm_fr_r}
       {boneName=deform_upperarm_up_base_r}
       {boneName=deform_upperarm_up_r}
      ]}]}
   n16 = Get HeadUpFloat() -> (HeadUpFloat:double)
   n17 = "Transform (Modify) Bone - Bone: base_hair_back_r"!(ComponentPose=Local To Component(LocalPose=@n15), Alpha=@n16, Rotation="0,-80,0") -> (Pose) {
    rotationMode=BMM_Additive
    rotationSpace=BCS_BoneSpace
    alphaBoolBlend {blendOption=Linear}
    boneToModify {boneName=base_hair_back_r}
    rotation {yaw=-80}
   }
   n18 = "Transform (Modify) Bone - Bone: base_hair_back_m"!(ComponentPose=@n17, Alpha=@n16, Rotation="0,-60,0") -> (Pose) {
    rotationMode=BMM_Additive
    rotationSpace=BCS_BoneSpace
    alphaBoolBlend {blendOption=Linear}
    boneToModify {boneName=base_hair_back_m}
    rotation {yaw=-60}
   }
   n19 = "Transform (Modify) Bone - Bone: base_hair_back_l"!(ComponentPose=@n18, Alpha=@n16, Rotation="0,-60,0") -> (Pose) {
    rotationMode=BMM_Additive
    rotationSpace=BCS_BoneSpace
    alphaBoolBlend {blendOption=Linear}
    boneToModify {boneName=base_hair_back_l}
    rotation {yaw=-60}
   }
   n20 = "Transform (Modify) Bone - Bone: base_hair_back_r"(ComponentPose=@n19, Alpha=@n14, Rotation="-9.548,16.558,-42.051") {
    rotationMode=BMM_Additive
    rotationSpace=BCS_BoneSpace
    alphaBoolBlend {blendOption=Linear}
    boneToModify {boneName=base_hair_back_r}
    rotation=<rot -9.548 16.558 -42.051>
   }
   n21 = "Transform (Modify) Bone - Bone: base_hair_back_m"(ComponentPose=@n20, Alpha=@n14, Rotation="-21.060,6.367,-25.587") {
    rotationMode=BMM_Additive
    rotationSpace=BCS_BoneSpace
    alphaBoolBlend {blendOption=Linear}
    boneToModify {boneName=base_hair_back_m}
    rotation=<rot -21.06 6.367 -25.587>
   }
   n22 = "Transform (Modify) Bone - Bone: base_hair_back_l"(ComponentPose=@n21, Alpha=@n14, Rotation="-24.383,-8.227,-23.316") {
    rotationMode=BMM_Additive
    rotationSpace=BCS_BoneSpace
    alphaBoolBlend {blendOption=Linear}
    boneToModify {boneName=base_hair_back_l}
    rotation=<rot -24.383 -8.227 -23.316>
   }
   n23 = "Transform (Modify) Bone - Bone: base_hair_back_r"(ComponentPose=@n22, Alpha=@n13, Rotation="14.467,-7.489,4.367") {
    rotationMode=BMM_Additive
    rotationSpace=BCS_BoneSpace
    alphaBoolBlend {blendOption=Linear}
    boneToModify {boneName=base_hair_back_r}
    rotation=<rot 14.467 -7.489 4.367>
   }
   n24 = "Transform (Modify) Bone - Bone: base_hair_back_m"(ComponentPose=@n23, Alpha=@n13, Rotation="12.226,2.246,30.221") {
    rotationMode=BMM_Additive
    rotationSpace=BCS_BoneSpace
    alphaBoolBlend {blendOption=Linear}
    boneToModify {boneName=base_hair_back_m}
    rotation=<rot 12.226 2.246 30.221>
   }
   n25 = "Transform (Modify) Bone - Bone: base_hair_back_l"(ComponentPose=@n24, Alpha=@n13, Rotation="6.411,8.559,32.583") {
    rotationMode=BMM_Additive
    rotationSpace=BCS_BoneSpace
    alphaBoolBlend {blendOption=Linear}
    boneToModify {boneName=base_hair_back_l}
    rotation=<rot 6.411 8.559 32.583>
   }
   n26 = "Transform (Modify) Bone - Bone: base_hair_back_r"(ComponentPose=@n25, Alpha=1, Rotation="0,10,15") {
    rotationMode=BMM_Additive
    rotationSpace=BCS_BoneSpace
    alphaBoolBlend {blendOption=Linear}
    boneToModify {boneName=base_hair_back_r}
    rotation {roll=15, yaw=10}
   }
   n27 = "Transform (Modify) Bone - Bone: base_hair_back_l"(ComponentPose=@n26, Alpha=1, Rotation="0,0,-8") {
    rotationMode=BMM_Additive
    rotationSpace=BCS_BoneSpace
    alphaBoolBlend {blendOption=Linear}
    boneToModify {boneName=base_hair_back_l}
    rotation {roll=-8}
   }
   Save cached pose 'Cache_StartPose'(Pose=Component To Local(ComponentPose=@n27)) {cacheName=Cache_StartPose}
  ]}
 EventGraph {nodes [
   Event BlueprintUpdateAnimation()
   > n1 = Sequence!(execute) -> (then_0, then_1, then_2:Exec, then_3:Exec, then_4:Exec, then_5:Exec, then_6:Exec, then_7:Exec, then_8:Exec, then_9:Exec)
   n2 = Branch(execute=@n1.then_1, Condition=Get bIsSkyDiving())
   > Set Linear Scale Free(Linear Scale Free:Vector="0.35,0.25,0.25") -> (Output_Get:Vector)
   > Set Linear Scale Backpack(Linear Scale Backpack:Vector="0.3,0.2,0.2") -> (Output_Get:Vector)
   > Set Linear Scale Cape(Linear Scale Cape:Vector="0.25,0.2,0.2") -> (Output_Get:Vector)
   Set Linear Scale Free(execute=@n2.else, Linear Scale Free="0.3,0.3,0.3")
   > Set Linear Scale Backpack(Linear Scale Backpack="0.2,0.2,0.2")
   > Set Linear Scale Cape(Linear Scale Cape="0.25,0.2,0.2")
   n3 = Gravity Override State Function DEPRECATED(
    execute=@n1.then_0
    CharacterPartAnimInstance=Self-Reference()
    GameTimeInSeconds=GetGameTimeInSeconds()
    GravityOverrideParams=Get Gravity Override Params()
   )
   > Set GravityOverride(GravityOverride=@n3)
   n4 = CheckForBackbling()
   > Set isWearingBackbling(isWearingBackbling=AND Boolean!(A=@n4, B=NOT Boolean(A=@n4.bIsBackblingACape), C:bool=NOT Boolean(A=@n4.bIsBackblingLong)) -> (ReturnValue))
   > Set Is Wearing Cape(Is Wearing Cape=OR Boolean(A=@n4.bIsBackblingACape, B=@n4.bIsBackblingLong))
   n5 = Get Skel Mesh()
   Set Head Down(
    execute=@n1.then_4
    Head Down:double=GetAlphaFromReferenceTransformDelta(SkeletalMeshComponent=@n5, BoneName=spine_05, BaseName="head ", InputRange=(X=0,Y=-90), OutputRange=(X=0,Y=-1), TwistAxis="0,0,1")
   ) -> (Output_Get:double)
   > Set Head_Turn_L(Head_Turn_L:double=GetAlphaFromReferenceTransformDelta(SkeletalMeshComponent=@n5, BoneName=spine_05, BaseName=head, DeltaComponentType=EulerX, InputRange=(X=0,Y=-22), TwistAxis="1,0,0")) -> (Output_Get:double)
   > Set Head_Turn_R(Head_Turn_R:double=GetAlphaFromReferenceTransformDelta(SkeletalMeshComponent=@n5, BoneName=spine_05, BaseName=head, DeltaComponentType=EulerX, InputRange=(X=0,Y=22), TwistAxis="1,0,0")) -> (Output_Get:double)
   n6 = GetAlphaFromReferenceTransformDelta(SkeletalMeshComponent=@n5, BoneName=spine_05, BaseName=head, InputRange=(X=10,Y=70), TwistAxis="0,0,1")
   Set HeadUpFloat(execute=@n1.then_3, HeadUpFloat:double=@n6) -> (Output_Get:double)
   n7 = float * float(A=@n6, B=3) {is=title K2Node_PromotableOperator}
   > n8 = Set HeadUp(HeadUp:bool="To Boolean (Integer)"!(InInt:int=Truncate(A=@n7)) -> (ReturnValue:bool)) -> (Output_Get:bool)
   n9 = NOT Boolean(A=Get bIsCrouching())
   n10 = Get Pistol Equipped() -> (Pistol Equipped:bool)
   > Set AimingUp_ShortW(AimingUp_ShortW:double="To Float (Boolean)"(InBool=AND Boolean(A=@n8, B=@n10, C=@n9))) -> (Output_Get:double)
   n11 = NOT Boolean(A=Get bIsMoving2D())
   n12 = Get LongW Equipped() -> (LongW Equipped:bool)
   n13 = float * float(A=@n6, B=2)
   n14 = AND Boolean(A=OR Boolean(A=Get isWearingBackbling(), B=Get Is Wearing Cape()), B="To Boolean (Integer)"(InInt=Truncate(A=@n13)), C=Get bIsCrouching())
   n15 = Branch(execute=@n1.then_2, Condition=AND Boolean(A=@n14, B=@n12, C=@n11))
   > Delay(Duration=0.4)
   > Set Crouching_BackpackCape_LW(Crouching_BackpackCape_LW:double=1) -> (Output_Get:double)
   Set Crouching_BackpackCape_LW(execute=@n15.else, Crouching_BackpackCape_LW=0)
   > n16 = Branch(Condition=AND Boolean(A=@n14, B=NOT Boolean(A=@n12), C=@n11))
   > Delay(Duration=0.4)
   > Set Crouching_BackpackCape_SW(Crouching_BackpackCape_SW:double=1) -> (Output_Get:double)
   Set Crouching_BackpackCape_SW(execute=@n16.else, Crouching_BackpackCape_SW=0)
   n17 = Cast To FortPlayerAnimInstance(execute=@n1.then_5, Object=GetAnimInstance(self=GetOwnerSkeletalMeshComponent()))
   > Set LongW Equipped(LongW Equipped:bool=false) -> (Output_Get:bool)
   > n18 = Switch on EFortWeaponCoreAnimation!(execute, Selection:EFortWeaponCoreAnimation Enum=Get WeaponCoreAnim(self=@n17)) -> (
    Melee:Exec
    Pistol:Exec
    Shotgun:Exec
    PaperBlueprint:Exec
    Rifle:Exec
    MeleeOneHand:Exec
    MachinePistol:Exec
    RocketLauncher:Exec
    GrenadeLauncher:Exec
    GoingCommando:Exec
    AssaultRifle:Exec
    TacticalShotgun:Exec
    SniperRifle:Exec
    TrapPlacement:Exec
    ShoulderLauncher:Exec
    AbilityDecoTool:Exec
    Crossbow:Exec
    C4:Exec
    RemoteControl:Exec
    DualWield:Exec
    AR_BullPup:Exec
    AR_ForwardGrip:Exec
    MedPackPaddles:Exec
    SMG_P90:Exec
    AR_DrumGun:Exec
    Consumable_Small:Exec
    Consumable_Large:Exec
    Balloon:Exec
    MountedTurret:Exec
    CreativeTool:Exec
    ExplosiveBow:Exec
    AshtonIndigo:Exec
    AshtonChicago:Exec
    MeleeDualWield:Exec
    Unarmed:Exec
   )
   n19 = Branch(
    execute=[@n18.AssaultRifle, @n18.GoingCommando, @n18.GrenadeLauncher, @n18.Rifle, @n18.Shotgun, @n18.SniperRifle, @n18.TacticalShotgun]
    Condition=Get bIsTargeting()
   )
   > Set LongW Equipped(LongW Equipped=true)
   Set LongW Equipped(execute=@n19.else, LongW Equipped=false)
   n20 = Branch(execute=[@n18.DualWield, @n18.Pistol], Condition=Get bIsTargeting())
   > Set Pistol Equipped(Pistol Equipped:bool=true) -> (Output_Get:bool)
   Set Pistol Equipped(execute=@n20.else, Pistol Equipped=false)
   Event OnCharacterCustomizationLoaded()
   > n21 = Sequence!(execute) -> (then_0, then_1)
   Delay(execute=@n21.then_1, Duration=0.2)
   > ResetDynamics(InTeleportType=TeleportPhysics)
   > ResetDynamics(InTeleportType=ResetPhysics)
   n22 = Cast To PlayerPawn_Athena(execute=@n21.then_0, Object=TryGetPawnOwner())
   > Set Body_Mesh(Body_Mesh=GetSkeletalMeshForPartType(self=@n22))
   Event BlueprintInitializeAnimation()
   > Set Skel Mesh(Skel Mesh=GetOwnerSkeletalMeshComponent())
  ]}
]
variables [
 AimingUp_ShortW {type=number}
 Body_Mesh {type=SkeletalMeshComponent}
 Crouching_BackpackCape_LW {type=number}
 Crouching_BackpackCape_SW {type=number}
 Gravity Override Params {
  type=GravityOverrideParamsStruct
  default {
   skydive_FloatingDown {gravityOverride=<0 0 -400>, windAmplitude=<200 -300 200>, windFrequency=<1.5 5 6>, windOffset=<0 -300 200>}
   skydive_GoingUp {gravityOverride=<0 -300 100>, windOffset=<0 0 200>}
  }
 }
 GravityOverride {type=Vector}
 Head Down {type=number}
 Head_Turn_L {type=number}
 Head_Turn_R {type=number}
 HeadUp {type=boolean}
 HeadUpFloat {type=number}
 Is Wearing Cape {type=boolean}
 isParachuting_alpha {type=number}
 isWearingBackbling {type=boolean}
 Linear Scale Backpack {type=Vector}
 Linear Scale Cape {type=Vector}
 Linear Scale Free {type=Vector}
 LongW Equipped {type=boolean}
 Pistol Equipped {type=boolean}
 removeLW {type=number}
 removeSW {type=number}
 Skel Mesh {type=SkeletalMeshComponent}
]
